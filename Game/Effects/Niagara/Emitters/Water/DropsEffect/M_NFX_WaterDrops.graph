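# Translucent subsurface material for Niagara water-drop sprites, with a static switch choosing drop or round mask and normal textures.
# Particle alpha scales the mask; the speed-based blur lerp blends identical samples, so it has no effect.
# Particle or fallback color gets tint, optional time-of-day and masked biome tint, then an HSV offset shared with the emissive branch.
# An optional switch adds water-surface displacement plus a Z offset.
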Material /Game/Effects/Niagara/Emitters/Water/DropsEffect/M_NFX_WaterDrops
graph {
 properties {
  bUsedWithNanite=true
  bUsedWithNiagaraSprites=true
  blendMode=BLEND_Translucent
  pixelDepthOffsetMode=PDOM_Legacy
  shadingModel=MSM_Subsurface
  translucencyLightingMode=TLM_VolumetricPerVertexDirectional
  usedShadingModels=Subsurface
 }
 nodes [
  n1 = TextureSample() {texture=/Game/VisualThreatManager/StormVisuals/Textures/T_Water_drop_02}
  n2 = ScalarParameter() {defaultValue=100, parameterName=ParticleSpeed}
  n3 = DynamicParameter() {paramNames [BlurNeedSpeedScale, Param2, Param3, Param4]}
  n4 = Saturate(None=Divide(A=ParticleSpeed(), B=Multiply(A=@n2, B=@n3.BlurNeedSpeedScale)))
  n5 = ShadingPathSwitch(Default=LinearInterpolate(A=@n1.R, B=@n1.R, Alpha=@n4), Mobile=@n1.R) {desc=Migrated from FeatureLevelSwitch}
  n6 = TextureSample() {texture=/Game/Global/Textures/Default/Masks_and_Gradients/T_SparkRound}
  n7 = ShadingPathSwitch(Default=LinearInterpolate(A=@n6.R, B=@n6.R, Alpha=@n4), Mobile=@n6.R) {desc=Migrated from FeatureLevelSwitch}
  n8 = ParticleColor()
  Reroute(None=Multiply(A=@n8.A, B=StaticSwitchParameter(True=@n5, False=@n7) {parameterName=SwitchShape}))
  n9 = VectorParameter() {group=Cascade, parameterName="Color_Morning (V3)", defaultValue=<rgba 1 1 1 0>}
  n10 = VectorParameter() {group=Cascade, parameterName="Color_Day (V3)", defaultValue=<rgba 1 1 1 0>}
  n11 = VectorParameter() {group=Cascade, parameterName="Color_Evening (V3)", defaultValue=<rgba 1 1 1 0>}
  n12 = VectorParameter() {group=Cascade, parameterName="Color_Night (V3)", defaultValue=<rgba 1 1 1 0>}
  n13 = ScalarParameter() {
   desc="Interpret Particle RGB as HSV Offsets, and use ParticleColor Fallback instead.  \r\n!! May light up Red in the preview as default ParticleColor = 1,1,1 - do not be alarmed."
   parameterName=Use Particle Color RGB as HSV Offset
  }
  n14 = Multiply(
   A=LinearInterpolate(
    A=@n8.RGB
    B=VectorParameter() {desc=fallback for RGB only, parameterName=ParticleColor RGB fallback, defaultValue=<rgba 1 1 1 1>}.RGB
    Alpha=@n13
   )
   B=VectorParameter() {desc=overall tint, parameterName=Base SSS Tint, defaultValue=<rgba 1 1 1 0>}.RGB
  )
  n15 = StaticSwitchParameter(
   True=Multiply(
    A=MaterialFunctionCall(Color_Morning=@n9.RGB, Color_Day=@n10.RGB, Color_Evening=@n11.RGB, Color_Night=@n12.RGB) {materialFunction=&Color_Time_of_Day}.Result
    B=@n14
   )
   False=@n14
  ) {group=Cascade, parameterName=UseTOD?}
  n16 = Multiply(A=VectorParameter() {group=WaterShader, parameterName=BiomeColorTint, defaultValue=<rgba 0.25 0.2418 0.125 2>}.RGB, B=@n15)
  n17 = MaterialFunctionCall() {materialFunction=/Game/Effects/Niagara/Emitters/Water/FoamEffect/MF_NFX_BiomeMask}
  n18 = ScalarParameter() {
   desc="Interpret Particle RGB as HSV Offsets, and use ParticleColor Fallback instead.  \r\n!! May light up Red in the preview as default ParticleColor = 1,1,1 - do not be alarmed."
   parameterName=Use Particle Color RGB as HSV Offset
  }
  n19 = LinearInterpolate(A=Constant3Vector(), B=ParticleColor().RGB, Alpha=@n18)
  Reroute(None=MaterialFunctionCall!(RGB=LinearInterpolate(A=@n15, B=@n16, Alpha=@n17.Result), HSV Offset=@n19, Optional Reference Color) {materialFunction=/Game/Effects/Niagara/Emitters/Water/FoamEffect/MF_HSV_Offset}.Result)
  n20 = MaterialFunctionCall!(EmissiveOvrride) {materialFunction=/Game/Effects/Niagara/Emitters/Water/FoamEffect/MF_NFX_EmissiveMask}
  Reroute(None=MaterialFunctionCall(
    RGB=MaterialFunctionCall(Emissive=@n20.Result) {materialFunction=&MF_FX_EmissiveAdjustments}.Result
    HSV Offset=@n19
   ) {materialFunction=&MF_HSV_Offset}.Result)
  StaticSwitchParameter(
   True=TextureSample() {samplerType=SAMPLERTYPE_Normal, texture=/Game/VisualThreatManager/StormVisuals/Textures/T_Water_drop_N_02}.RGB
   False=TextureSample() {samplerType=SAMPLERTYPE_Normal, texture=&T_Sphere_N}.RGB
  ) {parameterName=SwitchShape}
  n21 = MaterialFunctionCall!(Ignore XY Offsets=StaticBoolParameter() {defaultValue=true, parameterName=Ignore XY Gerstner Offsets}, Position, WaterBodyIndex, Depth) {materialFunction=/FortWater/Materials/Functions/WaterParticleDisplacement}
  n22 = VectorParameter() {parameterName=Water Displacement Multiplier, defaultValue=<rgba 1 1 1 0>}
  n23 = ScalarParameter() {defaultValue=1, parameterName=ZOffset}
  StaticSwitchParameter(
   True=Add(
    A=Multiply(A=@n21.WPO, B=@n22.RGB)
    B=MaterialFunctionCall(Z=@n23) {materialFunction=&MakeFloat3}.Result
   )
   False=Constant()
  ) {parameterName=UseWaterDisplacement}
 ]
}
